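# Conversation database for the NPC Blaze, defining her greeting and the services she offers when a player talks to her.
# A single entry point leads to a Speech node with a default line plus contextual lines for returning players, first-ever conversations, a specific outfit and a Valentine quest tag.
# Curve-table rows set each contextual line's priority.
# The speech then links by tag to the shared NPC quests and basic services conversations.
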ConversationDatabase /NPCLibrary/NPCs/Blaze/Blaze/NPCLibrary_Conversation_Blaze
entry_points [{tag=Conversation.Entry.Blaze, destinations [#1]}]
nodes [
 ConversationEntryPointNode_7 {
  id=#1
  class=ConversationEntryPointNode
  outputs [#2]
  properties {entryTag {tagName=Conversation.Entry.Blaze}}
 }
 Speech_C_4 {
  id=#2
  class=Speech_C
  outputs [#3, #4]
  properties {
   nodeName=""
   generalConfig {
    defaultMessage="Watch your step and you won't get <Hot>burned</>."
    contextualMessages [
     {
      message="Just couldn't stay away, huh?"
      contextRequirements [{requirement {class=FortControllerRequirement_CheckTimeSinceLastPlayedMatch}}]
      priority {curve {rowName=Tandem.Speech.Priority.WhileSincePlayed}}
     }
     {
      message=Let's get one thing straight: I don't do chit-chat.
      contextRequirements [{requirement {class=FortControllerRequirement_IsFirstEverConversationWithNPC}}]
      priority {curve {rowName=Tandem.Speech.Priority.FirstEverInteraction}}
     }
     {
      message="You've got some darkness in you, don't you girl?"
      contextRequirements [{requirement {
         class=FortControllerRequirement_HasCID
         properties {allowedCIds [/BRCosmetics/Athena/Items/Cosmetics/Characters/CID_963_Athena_Commando_F_Lexa]}
        }}]
      priority {curve {rowName=Tandem.Speech.Priority.OutfitReaction}}
     }
     {
      message=Well doesn't that just warm the heart.
      contextRequirements [{requirement {class=FortControllerRequirement_HasOwnedTag, properties {ownedTag {tagName=Athena.Quests.Crush.GiveValentine}}}}]
      priority {curve {rowName=Tandem.Speech.Priority.HasQuestGameplayTag}}
     }
    ]
   }
  }
 }
 ConversationLinkNode_0 {id=#4, class=ConversationLinkNode, properties {remoteEntryTag {tagName=Conversation.Entry.BasicServices}}}
 ConversationLinkNode_1 {id=#3, class=ConversationLinkNode, properties {remoteEntryTag {tagName=Conversation.Entry.NPCLibrary.Quests}}}
]
